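# Physics asset for the Bulletproof Tires skeletal mesh in the BulletproofTire valet mod.
# It holds a single body on root_jnt with no collision shapes listed.
# It defines no constraints, so there are no jointed parts to simulate.
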PhysicsAsset /ValetMods/BulletproofTire/Meshes/Bulletproof_Tires_SK_Physics
bodies {root_jnt []}
constraints []
